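; AI bot controller for the Puppeteer Demon boss that manages how it switches targets.
; On BeginPlay, after the parent logic, it binds a handler to the perception event fired when the current target changes.
; Each target change sets CanSwitchTarget to false on the boss's AI random targeter component.
; Nothing in this controller re-enables switching.
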(event EventBeginPlay
  (|CallParentFunction)
  (Perception|BindEventtoOnCurrentTargetChangedEvent (delegate Custom|TargetJustChanged)))

(event Custom|TargetJustChanged
  (params (OldTarget :type Actor%20Object%20Reference) (NewTarget :type Actor%20Object%20Reference))
  (Variables|/PuppeteerBoss/Pawn/FortControllerComponent_AIRandomTargeter_PuppeteerDemon.FortControllerComponent_AIRandomTargeter_PuppeteerDemon_C|SetCanSwitchTarget false (Variables|Default|GetFortControllerComponent_AIRandomTargeter_PuppeteerDemon)))

(fn ConstructionScript
  (|CallParentFunction))
