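; AI pawn for the Puppeteer Demon boss, which fights alongside a spawned mask actor and hangs from it in phase 2.
; On authority it scales health by playlist team size, spawns the mask, and maps gameplay tags to replicated battle stages driving music.
; First DBNO triggers phase 2, where damage only lands during mask staggers that end on a damage threshold or timer.
; Spawn-cloud resets reattach it to the mask; a calendar event force-kills it.
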(event EventBeginPlay
  (bind (_returnvalue_x _returnvalue_y) (Math|Vector|BreakVector (* (Math|Vector|GetForwardVector (Transformation|GetActorRotation)) (* (ScalableFloat|GetValueatLevel (Variables|Hotfix|GetMaskXYOffset_Hf)) -1.0))))
  (bind _leashvolume (Variables|Default|GetLeashVolume))
  (bind (_spawnlocation_x _spawnlocation_y _spawnlocation_z) (Math|Vector|BreakVector (Variables|GetSpawnLocation_X)))
  ;[0.03729 0.03327 0.15 0.5] BeginPlay
  (Variables|Default|SetIsStartPhase1 false)
  (Variables|Default|SetIsStartPhase2 false)
  ;;[1 1 1] For AnimBP
  (Components|SkeletalMesh|AnimationBlueprintLinking|LinkAnimClassLayers (Variables|Character|GetMesh) "/PuppeteerBoss/PuppeteerDemon_AnimBP.PuppeteerDemon_AnimBP_C")
  (EventDispatchers|BindEventtoOnDied (delegate Custom|Died))
  (|CallParentFunction)
  (Networking|SwitchHasAuthority
    (:Authority
      (bind _asbp_bot_controller_boss_puppeteer_demon (Utilities|Casting|CastToBP_BotController_Boss_PuppeteerDemon (Pawn|GetController))
        (:then
          (Variables|Default|SetDemonController _asbp_bot_controller_boss_puppeteer_demon))
        (:CastFailed
          _))
      (bind (_gamestate _playlist) (Class|FortAsyncActionCurrentPlaylistReady|PlaylistDataReadyAsync)
        (:OnReady
          (Utilities|FlowControl|Sequence
            (:then_0
              (Variables|SetSetByCallerReviveHealth_Value (Utilities|Struct|MakeScalableFloat (* (ScalableFloat|GetValueatLevel (Variables|Default|GetPlaylistHealthValues_Hf) (Math|Conversions|ToFloat(Integer) (CallFunction|GetMaxTeamSizefromPlaylist _playlist))) 1e+02)))
              (if (not (Variables|Default|GetIsSkyMangoVariant))
                (Health|SetHealthInvulnerabilityRatio 0.01))
              (Utilities|FlowControl|Delay 0.2)
              (bind _asfort_athena_aibot_controller (Utilities|Casting|CastToFortAthenaAIBotController (Pawn|GetController)))
              (Utilities|IsValid _leashvolume
                (:Is%20Valid
                  (Class|FortAthenaAIBotController|SetLeashVolume _asfort_athena_aibot_controller _leashvolume))
                (:Is%20Not%20Valid
                  _))
              (bind _returnvalue (Game|SpawnActorfromClass "/PuppeteerBoss/Mask/BGA_PuppeteerMask.BGA_PuppeteerMask_C" (Math|Transform|MakeTransform (Math|Vector|MakeVector (+ _returnvalue_x _spawnlocation_x) (+ _returnvalue_y _spawnlocation_y) (+ _spawnlocation_z (ScalableFloat|GetValueatLevel (Variables|Hotfix|GetMaskSpawnZOffset_Hf)))) (Transformation|GetActorRotation)) "Undefined" "MultiplyWithRoot" 0 0 self))
              (bind _output_get (Variables|Default|SetMaskActor _returnvalue))
              (Building|InitializeKismetSpawnedBuildingActor _output_get 0 0 false 0 false)
              (Default|CallMaskActorSpawned _output_get))
            (:then_1
              ;;[0.01794 0.1189 0.15 0.5] Update Battle State For Music
              (Ability|Tasks|WaitGameplayTagAddToActor 0 "(TagName=\"FirePetalBosses.Demons.IsinSpawnCloud\")"
                (:Added
                  (Variables|Default|SeteNewBattleStage "NewEnumerator0"))))
            (:then_2
              (Ability|Tasks|WaitGameplayTagAddToActor 0 "(TagName=\"FirePetalBosses.Demons.IsChallenged\")"
                (:then
                  (Utilities|FlowControl|Sequence
                    (:then_0
                      (Ability|Tasks|WaitGameplayTagAddToActor 0 "(TagName=\"PuppeteerBoss.Demon.IsInPhase2\")"
                        (:Added
                          (Variables|Default|SeteNewBattleStage "NewEnumerator2")
                          (CallFunction|SeverRoseDepthStartPhase2))))))
                (:Added
                  (if (GameplayTags|HasMatchingGameplayTag "(TagName=\"PuppeteerBoss.Demon.IsInPhase2\")")
                    (Variables|Default|SeteNewBattleStage "NewEnumerator2")
                    (else
                      (Variables|Default|SeteNewBattleStage "NewEnumerator1")
                      ;;[1 1 1] Send to the mask pawn
                      (CallFunction|SeverRoseDepthStartPhase1)
                      (Utilities|FlowControl|DelayUntilNextTick))))))
            (:then_3
              (Class|FortAsyncActionAuthorityMatchReady|MatchReadyAsync
                (:OnReady
                  (MCPContext|InitialCalendarSyncCompleted
                    (:OnSyncComplete
                      (MCPContext|WatchCalendarEvent (Variables|Default|GetCalendarEvent)
                        (:OnEventActive
                          (Class|BPFLAICommonFunctions|AddUniqueGameplayTag self "(TagName=\"PuppeteerBoss.Demon.ForceEliminated\")")
                          (FortPawn|ForceKill))))))))))))))

(event Custom|Died
  (params (DamagedActor :type Actor%20Object%20Reference) (Damage :type Float%20%28single-precision%29) (InstigatedBy :type Controller%20Object%20Reference) (DamageCauser :type Actor%20Object%20Reference) (HitLocation :type Vector) (FHitComponent :type Primitive%20Component%20Object%20Reference) (BoneName :type Name) (Momentum :type Vector))
  ;[0.03729 0.03327 0.15 0.5] BeginPlay
  ;;[0.1142 0.01812 0.15 0.5 13] OnDied - disable the minimap UI immediately FORT-769042
  (Components|Activation|Deactivate (Variables|Default|GetAthenaSpecialActor)))

(event Custom|SeverRoseDepthStartPhase2
  ;[0.03729 0.03327 0.15 0.5] BeginPlay
  (Variables|Default|SetIsStartPhase2 true))

(event Custom|PuppeteerSimulateEventStart
  ;[0.03729 0.03327 0.15 0.5] BeginPlay
  (Class|BPFLAICommonFunctions|AddUniqueGameplayTag self "(TagName=\"PuppeteerBoss.Demon.ForceEliminated\")")
  (FortPawn|ForceKill))

(event Custom|SeverRoseDepthStartPhase1
  ;[0.03729 0.03327 0.15 0.5] BeginPlay
  ;;[1 1 1] Send to the mask pawn
  (Variables|Default|SetIsStartPhase1 true)
  (Networking|CanExecuteCosmeticEvents
    (:True
      (Default|CallRoseDepthStartPhase1)
      (Variables|FX|GetIdleFX
        (:then
          (Components|Activation|Activate (Variables|FX|GetIdleFX) true))
        (:else
          (bind _returnvalue (Effects|SpawnFXSystemAttached :FXSystemAsset "/PuppeteerBoss/FX/NS_PuppeteerBoss_Strings_V2.NS_PuppeteerBoss_Strings_V2" :AttachToComponent (Variables|Character|GetMesh) :AttachPointName "None" :Scale "1.000000,1.000000,1.000000" :LocationType "KeepRelativeOffset" :bAutoDestroy true :PoolingMethod "AutoRelease" :bAutoActivate true))
          (Variables|FX|SetIdleFX _returnvalue))))))

(event Custom|PitchMusicDown
  (bind _cbm_musicpitchdown (Variables|Default|GetCBM_MusicPItchDown))
  ;[0.03729 0.03327 0.15 0.5] BeginPlay
  (Audio|Modulation|ActivateControlBusMix _cbm_musicpitchdown)
  (Utilities|FlowControl|Delay 1.3)
  (Audio|Modulation|DeactivateControlBusMix _cbm_musicpitchdown))

(event EventTick
  (params (DeltaSeconds :type Float%20%28single-precision%29)))

(event EventOnDamageServer
  (params (Damage :type Float%20%28single-precision%29) (DamageTags :type Gameplay%20Tag%20Container%20Structure%20%28by%20ref%29) (Momentum :type Vector :default "0, 0, 0") (HitInfo :type Hit%20Result%20Structure%20%28by%20ref%29) (InstigatedBy :type Controller%20Object%20Reference) (DamageCauser :type Actor%20Object%20Reference) (EffectContext :type Gameplay%20Effect%20Context%20Handle%20Structure))
  ;[0.15 0.0149 0.005945 0.5] DamageHandling
  (|CallParentFunction Damage DamageTags Momentum HitInfo InstigatedBy DamageCauser EffectContext)
  (if (Variables|Default|GetMaskStaggered)
    (bind _output_get (Variables|Default|SetDmgDuringMaskStagger (+ (Variables|Default|GetDmgDuringMaskStagger) Damage)))
    (if (>= _output_get (Variables|Default|GetMaskStaggerDmgThreshold))
      (CallFunction|EndMaskStagger))))

(event EventOnDamagePlayEffects
  (params (Damage :type Float%20%28single-precision%29) (DamageTags :type Gameplay%20Tag%20Container%20Structure%20%28by%20ref%29) (Momentum :type Vector :default "0, 0, 0") (HitInfo :type Hit%20Result%20Structure%20%28by%20ref%29) (InstigatedBy :type Fort%20Pawn%20Object%20Reference) (DamageCauser :type Actor%20Object%20Reference) (EffectContext :type Gameplay%20Effect%20Context%20Handle%20Structure))
  (bind (_bblockinghit _binitialoverlap _time _distance _location _impactpoint _normal) (Collision|BreakHitResult HitInfo))
  (bind _returnvalue (Math|Conversions|RotationFromXVector _normal))
  ;[1 1 1] RicochetImpactFX
  (if (or (and (not (Variables|Default|GetMaskStaggered)) (Variables|Default|GetIsStartPhase2)) (Variables|FX|GetIsInDBNO))
    _
    (elif (Variables|Default|GetIsInSpawnCloud)
      _))
  (bind _ricochetimpactfx (Variables|FX|GetRicochetImpactFX)
    (:then
      (Transformation|SetWorldLocation _ricochetimpactfx _location false false)
      (Transformation|SetWorldRotation _ricochetimpactfx _returnvalue false false)
      (Components|Activation|Activate _ricochetimpactfx true))
    (:else
      (bind _returnvalue_2 (Effects|Components|ParticleSystem|SpawnFXAtLocation "/PuppeteerBoss/FX/NS_PuppeteerBoss_RicochetImpact.NS_PuppeteerBoss_RicochetImpact" _location _returnvalue))
      (bind _output_get (Variables|FX|SetRicochetImpactFX _returnvalue_2))))
  (Effects|Components|ParticleSystem|SetVectorParameter _ricochetimpactfx "ImpactFXDirection" _normal))

(event Custom|EndMaskStagger
  (bind _maskactor (Variables|Default|GetMaskActor))
  ;[0.15 0.0149 0.005945 0.5] DamageHandling
  (Variables|FX|SetIsInDBNO true)
  (Variables|Default|SetMaskStaggered false)
  (Variables|Default|SetMaskStaggerAnimStarted false)
  (Class|BGAPuppeteerMask|ResetEyeDmg _maskactor)
  (Variables|/PuppeteerBoss/Mask/BGA_PuppeteerMask.BGA_PuppeteerMask_C|SetEyesIgnoreWeaponCollision false _maskactor)
  (Variables|Default|SetDmgDuringMaskStagger 0.0)
  (GameplayCue|AddGameplayCueOnActor(Looping) _maskactor (Variables|/PuppeteerBoss/Mask/BGA_PuppeteerMask.BGA_PuppeteerMask_C|GetGC_MaskAggressive_Loop _maskactor) (Ability|GameplayCue|MakeGameplayCueParameters 0.0))
  (Class|BGAPuppeteerMask|RestartTeleportCooldown _maskactor)
  (CallFunction|UnhidingPawnIdleFX))

(event Custom|StartMaskStaggerTimer
  ;[0.15 0.0149 0.005945 0.5] DamageHandling
  (bind _returnvalue (Utilities|Time|SetTimerbyEvent (delegate Custom|MaskStaggerTimerElapsed) (ScalableFloat|GetValueatLevel (Variables|Hotfix|GetMaskStaggerTimeLimit_Hf))))
  (Variables|Default|SetMaskStaggerTimerHandle _returnvalue)
  (CallFunction|HidingPawnIdleFX))

(event Custom|UnhidingPawnIdleFX
  ;[0.15 0.0149 0.005945 0.5] DamageHandling
  (Networking|CanExecuteCosmeticEvents
    (:True
      (Effects|Components|ParticleSystem|SetBoolParameter (Variables|FX|GetIdleFX) "HidingFX" false))))

(event Custom|HidingPawnIdleFX
  ;[0.15 0.0149 0.005945 0.5] DamageHandling
  (Networking|CanExecuteCosmeticEvents
    (:True
      (Effects|Components|ParticleSystem|SetBoolParameter (Variables|FX|GetIdleFX) "HidingFX" true))))

(event Custom|MaskStaggerTimerElapsed
  ;[0.15 0.0149 0.005945 0.5] DamageHandling
  (if (Variables|Default|GetMaskStaggered)
    (CallFunction|EndMaskStagger)))

(event Custom|PlayWFCIdleMontageEvent
  (params (Montage :type Anim%20Montage%20Object%20Reference) (StartSection :type Name))
  (bind _maskactor (Variables|Default|GetMaskActor))
  (CallFunction|ActivateHeadWeakpoint false)
  (Animation|PlayAnimMontage Montage 1.0 StartSection)
  (Utilities|IsValid _maskactor
    (:Is%20Valid
      _)
    (:Is%20Not%20Valid
      (Utilities|FlowControl|Delay 3.0)))
  (Transformation|AttachActorToComponent (Variables|/PuppeteerBoss/Mask/BGA_PuppeteerMask.BGA_PuppeteerMask_C|GetDemonHolderSpawn _maskactor) "None" "SnapToTarget" "SnapToTarget" "KeepWorld" false)
  (Pawn|Components|CharacterMovement|SetMovementMode (Variables|Character|GetCharacterMovement) "MOVE_None" 0))

(event EventOnDBNOStateChanged_Event
  (params (FortPawn :type Fort%20Pawn%20Object%20Reference) (bInIsDBNO :type Boolean))
  (bind _idlefx (Variables|FX|GetIdleFX))
  ;[0.15 0.06321 0.08279 0.5] OnDBNO
  (|CallParentFunction FortPawn bInIsDBNO)
  (Networking|SwitchHasAuthority
    (:Authority
      (bind _output_get (Variables|FX|SetIsInDBNO bInIsDBNO))
      (if _output_get
        (Utilities|FlowControl|DoOnce false
          (:Completed
            (Variables|Default|SetDBNOBlockCarry true)
            (Class|BPFLAICommonFunctions|AddUniqueGameplayTag self "(TagName=\"FirePetalBosses.Demons.BlockReturnToSpawnCloud\")")
            (Utilities|FlowControl|Delay (Variables|Default|GetMaskSpawnDelayTime))
            (Class|BPFLAICommonFunctions|AddUniqueGameplayTag self "(TagName=\"PuppeteerBoss.Demon.IsInPhase2\")")))
        (else
          (Class|BPFLAICommonFunctions|AddUniqueGameplayTag self "(TagName=\"Athena.Gameplay.Status.ForceSkipDBNO\")"))))
    (:Remote
      (if (Variables|FX|GetIsInDBNO)
        (Effects|Components|ParticleSystem|SetBoolParameter _idlefx "HidingFX" true)
        (else
          (Effects|Components|ParticleSystem|SetBoolParameter _idlefx "HidingFX" false)
          (CallFunction|EnableNeverSkipAnim false) ; now allow skip frames as we'll be held by mask
        )))))

(event Custom|PlayExitSCMontageEvent
  (params (Montage :type Anim%20Montage%20Object%20Reference) (StartSection :type Name) (SkipAnim :type Boolean))
  (if SkipAnim
    (Utilities|FlowControl|Delay 0.5)
    (else
      (Animation|StopAnimMontage 0)
      (Transformation|DetachFromActor "KeepWorld" "KeepWorld" "KeepWorld")
      (Pawn|Components|CharacterMovement|SetMovementMode (Variables|Character|GetCharacterMovement) "MOVE_Walking" 0)
      (Animation|PlayAnimMontage Montage 1.0 StartSection)
      (bind _returnvalue (ContextualAnim|Utilities|GetSectionLength Montage (Montage|GetSectionIndex Montage StartSection)))
      (Utilities|FlowControl|Delay _returnvalue)))
  (Class|BPFLAICommonFunctions|RemoveAllInstancesOfGameplayTag self "(TagName=\"FirePetalBosses.Demons.IsinSpawnCloud\")")
  (CallFunction|ActivateHeadWeakpoint true)
  (CallFunction|ChallengeStarted))

(event Custom|DisableCarry
  ;[0.15 0.06321 0.08279 0.5] OnDBNO
  (bind _asfort_ability_system_component (Utilities|Casting|CastToFortAbilitySystemComponent (Ability|GetAbilitySystemComponent self)))
  (Tags|BlockAbilitiesWithTags _asfort_ability_system_component (Variables|Default|GetAbilityTagToBlock)))

(event Custom|PlayInactiveDemonMontageEvent
  (params (Montage :type Anim%20Montage%20Object%20Reference) (StartSection :type Name))
  (Animation|PlayAnimMontage Montage 1.0 StartSection))

(event EventOnCharacterCustomizationCompleted
  (params (Pawn :type Fort%20Player%20Pawn%20Object%20Reference))
  ;[1 1 1] Assign Boss Material Instances
  (|CallParentFunction Pawn)
  (Rendering|Material|SetMaterial (Customization|GetSkeletalMeshforPartType "Body") 0 "/PuppeteerBoss/Materials/MI_RoseDepth_DemonBoss_Body.MI_RoseDepth_DemonBoss_Body")
  (Rendering|Material|SetMaterial (Customization|GetSkeletalMeshforPartType "Head") 0 "/PuppeteerBoss/Materials/MI_RoseDepth_DemonBoss_Head.MI_RoseDepth_DemonBoss_Head")
  (Rendering|Material|SetMaterial (Customization|GetSkeletalMeshforPartType "Face") 1 "/PuppeteerBoss/Materials/MI_RoseDepth_DemonBoss_Eyes.MI_RoseDepth_DemonBoss_Eyes")
  (Rendering|Material|SetMaterial (Customization|GetSkeletalMeshforPartType "Face") 0 "/PuppeteerBoss/Materials/MI_RoseDepth_DemonBoss_FaceAcc.MI_RoseDepth_DemonBoss_FaceAcc"))

(event Custom|MaskPullRotate
  (bind (_returnvalue_roll _returnvalue_pitch _returnvalue_yaw) (Math|Rotator|BreakRotator (Math|Rotator|FindLookatRotation (Transformation|GetActorLocation (Variables|Default|GetMaskActor)) (Transformation|GetActorLocation))))
  (bind (_prepullrot_roll _prepullrot_pitch _prepullrot_yaw) (Math|Rotator|BreakRotator (Variables|GetPrePullRot_Roll)))
  (Variables|Default|SetPrePullRot (Transformation|GetActorRotation))
  (bind (_direction _alpha) (Timeline|MaskPullRot_Timeline PlayFromStart)
    (:Update
      (Transformation|SetActorRotation (Math|Rotator|Lerp(Rotator) (Math|Rotator|MakeRotator 0.0 0.0 _prepullrot_yaw) (Math|Rotator|MakeRotator 0.0 0.0 _returnvalue_yaw) _alpha true)))))

(event Custom|TargetSwitchNoNewTarget
  (bind _demoncontroller (Variables|Default|GetDemonController))
  (bind _fortcontrollercomponent_airandomtargeter_puppeteerdemon (Variables|/PuppeteerBoss/Pawn/BP_BotController_Boss_PuppeteerDemon.BP_BotController_Boss_PuppeteerDemon_C|GetFortControllerComponent_AIRandomTargeter_PuppeteerDemon _demoncontroller))
  (Utilities|FlowControl|RetriggerableDelay (ScalableFloat|GetValueatLevel (Variables|Default|GetTargetSwitchNoNewTargetDelay_Hf)))
  (if (Variables|/PuppeteerBoss/Pawn/FortControllerComponent_AIRandomTargeter_PuppeteerDemon.FortControllerComponent_AIRandomTargeter_PuppeteerDemon_C|GetCanSwitchTarget _fortcontrollercomponent_airandomtargeter_puppeteerdemon)
    (Variables|/PuppeteerBoss/Pawn/FortControllerComponent_AIRandomTargeter_PuppeteerDemon.FortControllerComponent_AIRandomTargeter_PuppeteerDemon_C|SetCanSwitchTarget false _fortcontrollercomponent_airandomtargeter_puppeteerdemon)))

(event Custom|StopMaskPullRotate
  (bind (_returnvalue_roll _returnvalue_pitch _returnvalue_yaw) (Math|Rotator|BreakRotator (Math|Rotator|FindLookatRotation (Transformation|GetActorLocation (Variables|Default|GetMaskActor)) (Transformation|GetActorLocation))))
  (bind (_prepullrot_roll _prepullrot_pitch _prepullrot_yaw) (Math|Rotator|BreakRotator (Variables|GetPrePullRot_Roll)))
  (Timeline|MaskPullRot_Timeline Stop))

(fn ActivateHeadWeakpoint
  (params (Activate :type Boolean))
  (DamageZones|UpdateBonesatDamageZoneType "Critical" (select Activate (Utilities|Array|MakeArray "head") (Utilities|Array|MakeArray "None"))))

(fn BlueprintModifyIncomingDamage
  (params (Damage :type Float%20%28single-precision%29) (DamageTags :type Gameplay%20Tag%20Container%20Structure%20%28by%20ref%29) (EffectContext :type Gameplay%20Effect%20Context%20Handle%20Structure%20%28by%20ref%29) (EventInstigator :type Controller%20Object%20Reference) (DamageCauser :type Actor%20Object%20Reference) (HitInfo :type Hit%20Result%20Structure))
  (bind _returnvalue (|CallParentFunction Damage DamageTags EffectContext EventInstigator DamageCauser HitInfo))
  (if (Variables|Default|GetIsSkyMangoVariant)
    _
    (elif (Variables|Default|GetIsInSpawnCloud)
      (bind _output_get (LocalVariables|SetNewDamage 0.0))
      (return _output_get)))
  (if (and (Variables|Default|GetMaskStaggered) (< (Variables|Default|GetDmgDuringMaskStagger) (Variables|Default|GetMaskStaggerDmgThreshold)))
    (bind _output_get_2 (LocalVariables|SetNewDamage _returnvalue))
    (return _output_get_2)
    (elif (GameplayTags|HasMatchingGameplayTag "(TagName=\"PuppeteerBoss.Demon.IsInPhase2\")")
      (bind _output_get_3 (LocalVariables|SetNewDamage 0.0))
      (return _output_get_3)
      (else
        (return _returnvalue)))))

(fn ChallengeStarted
  (bind _maskactor (Variables|Default|GetMaskActor))
  (|CallParentFunction)
  (Utilities|FlowControl|Sequence
    (:then_0
      (Health|SetHealthInvulnerabilityRatio 0.0)
      (Variables|Default|SetIsInSpawnCloud false)
      (Variables|Default|SetAlertLevel "Threatened")
      (Utilities|IsValid _maskactor
        (:Is%20Valid
          (Class|BPFLAICommonFunctions|RemoveAllInstancesOfGameplayTag _maskactor "(TagName=\"FirePetalBosses.Demons.IsinSpawnCloud\")")
          (if (GameplayTags|HasMatchingGameplayTag "(TagName=\"PuppeteerBoss.Demon.IsInPhase2\")")
            (Variables|/PuppeteerBoss/Mask/BGA_PuppeteerMask.BGA_PuppeteerMask_C|SetEyesIgnoreWeaponCollision false _maskactor)
            (Variables|/PuppeteerBoss/Mask/BGA_PuppeteerMask.BGA_PuppeteerMask_C|SetPauseTeleport false _maskactor)
            (Class|BGAPuppeteerMask|RestartTeleportCooldown _maskactor)))))
    (:then_1
      (Variables|/PuppeteerBoss/Mask/BGA_PuppeteerMask.BGA_PuppeteerMask_C|SetbShouldPlayIdleSound false (Variables|Default|GetMaskActor)))))

(fn EnableNeverSkipAnim
  (params (EnableNeverSkip :type Boolean))
  (if (ScalableFloat|GetValueasBool (Variables|Default|GetNeverSkipAnimTickWhileActive_Hf))
    (Anim|SetNeverSkipAnimTick EnableNeverSkip)))

(fn GetDataForDemonAnimBP
  (return (Variables|Default|GetMaskActor) (Variables|Default|GetMaskStaggered) (Variables|Default|GetIsInSpawnCloud)))

(fn GetDataForMaskAnimBP
  (return (Variables|Default|GetMaskStaggered) (Variables|Default|GetIsInSpawnCloud) (Variables|Default|GetIsStartPhase1)))

(fn GetMaxTeamSizeFromPlaylist
  (params (Playlist :type Fort%20Playlist%20Object%20Reference))
  (bind _output_get (LocalVariables|SetMax%20Team%20Size (Variables|/Script/FortniteGame.FortPlaylist|GetMaxTeamSize Playlist)))
  (if (< _output_get 1)
    (LocalVariables|SetMax%20Team%20Size 1)
    (elif (> (LocalVariables|GetMax%20Team%20Size) 4)
      (LocalVariables|SetMax%20Team%20Size 4)
      (return (LocalVariables|GetMax%20Team%20Size)))))

(fn InitiatedChallengeStart
  (bind _maskactor (Variables|Default|GetMaskActor))
  (|CallParentFunction)
  (Variables|Default|SetAlertLevel "Suspicious")
  (Variables|/PuppeteerBoss/Mask/BGA_PuppeteerMask.BGA_PuppeteerMask_C|SetbShouldPlayAwakeSound true _maskactor)
  (Variables|/PuppeteerBoss/Mask/BGA_PuppeteerMask.BGA_PuppeteerMask_C|SetbShouldPlayIdleSound false _maskactor)
  (Default|CallEDInitiatedChallengeStart))

(fn InitiatedTeleportToSpawnCloud
  (bind _maskactor (Variables|Default|GetMaskActor))
  (|CallParentFunction)
  (if (not (Variables|Default|GetIsSkyMangoVariant))
    (Health|SetHealthInvulnerabilityRatio 0.01))
  (Variables|FX|SetIsInDBNO true)
  (Variables|Default|SetIsPlayerOutOfArena true)
  (Variables|Default|SetMaskStaggered false)
  (Variables|Default|SetAlertLevel "LostTarget")
  (Utilities|IsValid _maskactor
    (:Is%20Valid
      (if (GameplayTags|HasMatchingGameplayTag _maskactor "(TagName=\"PuppeteerBoss.Demon.IsInPhase2\")")
        (Variables|/PuppeteerBoss/Mask/BGA_PuppeteerMask.BGA_PuppeteerMask_C|SetPauseTeleport true _maskactor)))))

(fn MaskStagger
  (if (> (FortPawn|GetHealth) (/ (FortPawn|GetMaxHealth) 2.0))
    (Variables|Default|SetMaskStaggerDmgThreshold (/ (FortPawn|GetHealth) 2.0))
    (else
      (Variables|Default|SetMaskStaggerDmgThreshold (+ (FortPawn|GetMaxHealth) 1e+02))))
  (Variables|Default|SetMaskStaggered true)
  (Variables|FX|SetIsInDBNO false)
  (Variables|Default|SetDmgDuringMaskStagger 0.0)
  (CallFunction|StartMaskStaggerTimer))

(fn OnRep_bMaskIsDead)

(fn OnRep_DBNOBlockCarry
  (if (Variables|Default|GetDBNOBlockCarry)
    (CallFunction|DisableCarry)))

(fn OnRep_eNewBattleStage
  (bind _enewbattlestage (Variables|Default|GeteNewBattleStage))
  (Networking|SwitchHasAuthority
    (:Remote
      (Class|NPCMusicComponentPuppeteerBoss|UpdateMusic (Variables|Default|GetNPCMusicComponent_PuppeteerBoss) _enewbattlestage)
      (switch Utilities|FlowControl|Switch|SwitchonePuppeteerCombatStage _enewbattlestage
        (:NewEnumerator0
          (bind _inairaudiocomp (Variables|Sounds|GetInAirAudioComp)
            (:then
              (Audio|Components|Audio|FadeIn _inairaudiocomp 1.0 1.0 0.0 "Linear" :!class /Script/Engine.AudioComponent))
            (:else)))
        (:NewEnumerator1
          _)
        (:NewEnumerator2
          (Utilities|FlowControl|Sequence
            (:then_0
              _)
            (:then_1
              (CallFunction|PitchMusicDown)))))
      (bind _inairaudiocomp_2 (Variables|Sounds|GetInAirAudioComp)
        (:then
          (if (Audio|Components|Audio|IsPlaying _inairaudiocomp_2)
            _))
        (:else
          _))
      (Audio|Components|Audio|FadeOut _inairaudiocomp_2 1.0 0.0 "Linear" :!class /Script/Engine.AudioComponent))))

(fn OnRep_IsInSpawnCloud
  (bind _maskactor (Variables|Default|GetMaskActor))
  (bind _returnvalue (GameplayTags|HasMatchingGameplayTag _maskactor "(TagName=\"PuppeteerBoss.Demon.IsInPhase2\")"))
  (bind _maskactor_2 (Variables|Default|GetMaskActor))
  (Networking|SwitchHasAuthority
    (:Authority
      (Utilities|IsValid _maskactor
        (:Is%20Valid
          (if _returnvalue
            (if (Variables|Default|GetIsInSpawnCloud)
              (Utilities|Time|ClearandInvalidateTimerbyHandle (Variables|/PuppeteerBoss/Mask/BGA_PuppeteerMask.BGA_PuppeteerMask_C|GetInitiateTeleportTimerHandle _maskactor_2))
              (else
                (Class|BGAPuppeteerMask|Phase2ResetInitiateTeleport _maskactor_2)))))))
    (:Remote
      (CallFunction|EnableNeverSkipAnim (and (not _returnvalue) (not (Variables|Default|GetIsInSpawnCloud)))) ; only should enable non tick when in spawn cloud and if not yet at phase 2 as in phase 2 demon will be hanging
    )))

(fn OnRep_MaskStaggered
  (Networking|SwitchHasAuthority
    (:Remote
      (CallFunction|EnableNeverSkipAnim (Variables|Default|GetMaskStaggered)))))

(fn PlayExitSCMontage
  (params (Montage :type Anim%20Montage%20Object%20Reference) (StartSection :type Name :default "None") (SkipAnim :type Boolean))
  (|CallParentFunction Montage StartSection SkipAnim)
  (CallFunction|PlayExitSCMontageEvent Montage StartSection SkipAnim))

(fn PlayWFCIdleMontage
  (params (Montage :type Anim%20Montage%20Object%20Reference) (StartSection :type Name :default "None"))
  (|CallParentFunction Montage StartSection)
  (CallFunction|PlayWFCIdleMontageEvent Montage StartSection))

(fn TeleportedToSpawnCloud
  (bind (_returnvalue_x _returnvalue_y) (Math|Vector|BreakVector (* (Math|Vector|GetForwardVector (Variables|Default|GetSpawnRotation)) (* (ScalableFloat|GetValueatLevel (Variables|Hotfix|GetMaskXYOffset_Hf)) -1.0))))
  (bind (_spawnlocation_x _spawnlocation_y _spawnlocation_z) (Math|Vector|BreakVector (Variables|GetSpawnLocation_X)))
  (bind _maskactor (Variables|Default|GetMaskActor))
  (|CallParentFunction)
  (Variables|Default|SetAlertLevel "Unaware")
  (Variables|Default|SetMaskStaggerAnimStarted false)
  (Variables|Default|SetIsPlayerOutOfArena false)
  (Variables|Default|SetDmgDuringMaskStagger 0.0)
  (Variables|Default|SetIsInSpawnCloud true)
  (Utilities|IsValid _maskactor
    (:Is%20Valid
      (Transformation|AttachActorToComponent (Variables|/PuppeteerBoss/Mask/BGA_PuppeteerMask.BGA_PuppeteerMask_C|GetDemonHolderSpawn _maskactor) "None" "SnapToTarget" "SnapToTarget" "KeepWorld" false)
      (Class|BGAPuppeteerMask|Teleport _maskactor (Math|Vector|MakeVector (+ _returnvalue_x _spawnlocation_x) (+ _returnvalue_y _spawnlocation_y) (+ _spawnlocation_z (ScalableFloat|GetValueatLevel (Variables|Hotfix|GetMaskSpawnZOffset_Hf)))) true)
      (Class|BGAPuppeteerMask|ResetEyeDmg _maskactor)
      (Utilities|FlowControl|Sequence
        (:then_0
          (Variables|/PuppeteerBoss/Mask/BGA_PuppeteerMask.BGA_PuppeteerMask_C|SetEyesIgnoreWeaponCollision true _maskactor)
          (if (not (GameplayTags|HasMatchingGameplayTag _maskactor "(TagName=\"FirePetalBosses.Demons.IsinSpawnCloud\")"))
            (Class|BPFLAICommonFunctions|AddUniqueGameplayTag _maskactor "(TagName=\"FirePetalBosses.Demons.IsinSpawnCloud\")")))
        (:then_1
          (Variables|/PuppeteerBoss/Mask/BGA_PuppeteerMask.BGA_PuppeteerMask_C|SetbShouldPlayIdleSound true (Variables|Default|GetMaskActor)))))))

(fn TrySwitchAggroTarget
  (Variables|/PuppeteerBoss/Pawn/FortControllerComponent_AIRandomTargeter_PuppeteerDemon.FortControllerComponent_AIRandomTargeter_PuppeteerDemon_C|SetCanSwitchTarget true (Variables|/PuppeteerBoss/Pawn/BP_BotController_Boss_PuppeteerDemon.BP_BotController_Boss_PuppeteerDemon_C|GetFortControllerComponent_AIRandomTargeter_PuppeteerDemon (Variables|Default|GetDemonController)))
  (CallFunction|TargetSwitchNoNewTarget))

(fn UpdateStrikeCollisionBox
  (bind _strikecollision (Variables|Default|GetStrikeCollision))
  (bind (_boxextent_x _boxextent_y) (Math|Vector|BreakVector (Variables|/Script/Engine.BoxComponent|GetBoxExtent_X _strikecollision)))
  (Components|Box|SetBoxExtent _strikecollision (Math|Vector|MakeVector (ScalableFloat|GetValueatLevel (Variables|Hotfix|GetStrikeCollisionBoxExtentX_Hf)) (ScalableFloat|GetValueatLevel (Variables|Hotfix|GetStrikeCollisionBoxExtentY_Hf)) (ScalableFloat|GetValueatLevel (Variables|Hotfix|GetStrikeCollisionBoxExtentZ_Hf))) false)
  (Transformation|AddRelativeLocation _strikecollision (Math|Vector|MakeVector 0.0 _boxextent_y) false false))

(fn UpdateStrikeCollisionSphere
  (bind _strikecollisionsphereradius_hf (Variables|Hotfix|GetStrikeCollisionSphereRadius_Hf))
  (bind _returnvalue (ScalableFloat|GetValueatLevel _strikecollisionsphereradius_hf))
  (bind _strikecollisionsphere (Variables|Default|GetStrikeCollisionSphere))
  (Components|Sphere|SetSphereRadius _strikecollisionsphere _returnvalue false)
  (Transformation|AddRelativeLocation _strikecollisionsphere (Math|Vector|MakeVector 0.0 0.0 (+ _returnvalue 5.0)) false false))

(fn ConstructionScript
  (|CallParentFunction)
  (CallFunction|UpdateStrikeCollisionBox)
  (CallFunction|UpdateStrikeCollisionSphere))
