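; AI ghost pawn for the Puppeteer boss fight that haunts a target player and can attack when damaged.
; On the server it leashes itself to its spawn area and syncs target, haunt tag and blackboard.
; Leaving the leash, knockback, target DBNO or death, or demon death despawns it with FX; ending play applies haunt cooldown.
; Replicated variables drive client audio and hide/unhide through timestamped custom primitive data.
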(event EventBeginPlay
  (|CallParentFunction)
  (Networking|SwitchHasAuthority
    (:Authority
      (bind _asfort_athena_aibot_controller (Utilities|Casting|CastToFortAthenaAIBotController (Pawn|GetController)))
      (Variables|Default|SetGhostController _asfort_athena_aibot_controller)
      (Utilities|FlowControl|Delay 0.2)
      (CallFunction|TieLeashtoGhostSpawnArea)
      (bind _returnvalue (Utilities|Time|SetTimerbyEvent (EventDispatchers|CreateEvent :!event Custom|CheckGhostIsInLeash) (ScalableFloat|GetValueatLevel (Variables|Hotfix|GetLeashCheckTimerPeriod_Hf)) true))
      (Variables|Default|SetLeashCheckTimerHandle _returnvalue)
      (Ability|Tasks|WaitGameplayTagAddToActor 0 "(TagName=\"Gameplay.Status.Knockback\")" true
        (:then
          (if (GameplayTags|HasMatchingGameplayTag "(TagName=\"PuppeteerBoss.Ghost.Seen\")")
            (Ability|Tasks|WaitGameplayTagRemoveFromActor 0 "(TagName=\"PuppeteerBoss.Ghost.Seen\")" true
              (:then
                (bind _returnvalue_2 (Utilities|Time|SetTimerbyEvent (delegate Custom|MaxTimeBeforeUnhideReached) (ScalableFloat|GetValueatLevel (Variables|Hotfix|GetMaxTimeBeforeFirstUnhide_Hf))))
                (Variables|Default|SetTimeBeforeUnhideTimer _returnvalue_2))
              (:Removed
                (Variables|Default|SetShouldHide false)
                (if (Utilities|Time|IsValidTimerHandle (Variables|Default|GetTimeBeforeUnhideTimer))
                  (Utilities|Time|ClearandInvalidateTimerbyHandle (Variables|Default|GetTimeBeforeUnhideTimer)))))
            (else
              (Variables|Default|SetShouldHide false))))
        (:Added
          (CallFunction|GhostAttackFinished))))))

(event Custom|MaxTimeBeforeUnhideReached
  (Actor|DestroyActor))

(event Custom|BindToDemonPawn
  (params (Demon%20Pawn :type BP%20Player%20Pawn%20Puppeteer%20Demon%20Object%20Reference))
  (bind _output_get (Variables|Default|SetDemonPawn Demon%20Pawn))
  (Health|BindEventtoOnPawnHealthChanged _output_get (EventDispatchers|CreateEvent :!event Custom|DemonPawnHealthChanged))
  (EventDispatchers|BindEventtoOnDied _output_get (EventDispatchers|CreateEvent :!event Custom|DemonPawnDied)))

(event EventOnCharacterCustomizationCompleted
  (params (Pawn :type Fort%20Player%20Pawn%20Object%20Reference))
  (bind _returnvalue (Customization|GetSkeletalMeshforPartType "Head"))
  ; PROTO_Material
  (|CallParentFunction Pawn)
  (Rendering|Material|SetMaterial _returnvalue 0 "/FirePetalBosses/Assets/Materials/MI_PuppeteerDemon_Ghost.MI_PuppeteerDemon_Ghost")
  (Rendering|SetRenderCustomDepth _returnvalue true))

(event Custom|PlayDisappearFX
  (bind _returnvalue (Transformation|GetActorLocation))
  (Networking|CanExecuteCosmeticEvents
    (:True
      (Effects|Components|ParticleSystem|SpawnFXAtLocation "/PuppeteerBoss/FX/NS_PuppeteerBoss_GhostDisappear.NS_PuppeteerBoss_GhostDisappear" _returnvalue)
      (Audio|SpawnSoundatLocation (Variables|Default|GetGhostDissapearSound) _returnvalue))))

(event Custom|GhostAttackFinished
  (CallFunction|PlayDisappearFX)
  (Variables|Audio|SetbGhostAudioShouldStop true)
  (Utilities|FlowControl|Delay 0.2)
  (Actor|DestroyActor))

(event Custom|ResetWaitForDBNOOrDeath
  (bind _targetplayer (Variables|Default|GetTargetPlayer))
  (Utilities|IsValid _targetplayer
    (:Is%20Valid
      (EventDispatchers|UnbindEventfromOnPlayerStartDBNO _targetplayer (delegate Custom|TargetIsDBNO))
      (EventDispatchers|UnbindEventfromOnDied _targetplayer (delegate Custom|TargetDied)))))

(event Custom|DemonPawnHealthChanged
  (bind _demonpawn (Variables|Default|GetDemonPawn))
  (Utilities|IsValid _demonpawn
    (:Is%20Valid
      (if (and (FortPawn|IsDead _demonpawn) (GameplayTags|HasMatchingGameplayTag _demonpawn "(TagName=\"PuppeteerBoss.Demon.IsInPhase2\")"))
        (CallFunction|PlayDisappearFX)
        (Variables|Audio|SetbGhostAudioShouldStop true)
        (Utilities|FlowControl|Delay 0.2)
        (Actor|DestroyActor)))))

(event Custom|WaitForDBNOOrDeath
  (bind _targetplayer (Variables|Default|GetTargetPlayer))
  (EventDispatchers|BindEventtoOnPlayerStartDBNO _targetplayer (delegate Custom|TargetIsDBNO))
  (EventDispatchers|BindEventtoOnDied _targetplayer (delegate Custom|TargetDied)))

(event Custom|DemonPawnDied
  (params (DamagedActor :type Actor%20Object%20Reference) (Damage :type Float%20%28single-precision%29) (InstigatedBy :type Controller%20Object%20Reference) (DamageCauser :type Actor%20Object%20Reference) (HitLocation :type Vector) (FHitComponent :type Primitive%20Component%20Object%20Reference) (BoneName :type Name) (Momentum :type Vector))
  (CallFunction|PlayDisappearFX)
  (Variables|Audio|SetbGhostAudioShouldStop true)
  (Utilities|FlowControl|Delay 0.2)
  (Actor|DestroyActor))

(event Custom|FadeOutGhostLoop
  ;[0.03466 0.09482 0.15 0.5] Stop Audio
  (bind _inairaudiocomp (Variables|Sounds|GetInAirAudioComp)
    (:then
      (Audio|Components|Audio|FadeOut _inairaudiocomp 0.2 0.0 "Linear" :!class /Script/Engine.AudioComponent))
    (:else)))

(event Custom|TargetIsDBNO
  (CallFunction|GhostAttackFinished))

(event Custom|TargetDied
  (params (DamagedActor :type Actor%20Object%20Reference) (Damage :type Float%20%28single-precision%29) (InstigatedBy :type Controller%20Object%20Reference) (DamageCauser :type Actor%20Object%20Reference) (HitLocation :type Vector) (FHitComponent :type Primitive%20Component%20Object%20Reference) (BoneName :type Name) (Momentum :type Vector))
  (CallFunction|GhostAttackFinished))

(event EventEndPlay
  (params (EndPlayReason :type EEndPlayReason%20Enum :default "Destroyed"))
  (|CallParentFunction EndPlayReason)
  (Networking|SwitchHasAuthority
    (:Authority
      (CallFunction|StopHauntingTarget (Variables|Default|GetTargetPlayer))
      (Utilities|Time|ClearandInvalidateTimerbyHandle (Variables|Default|GetLeashCheckTimerHandle)))
    (:Remote
      ;[0.03466 0.09482 0.15 0.5] Stop Audio
      (bind _inairaudiocomp (Variables|Sounds|GetInAirAudioComp)
        (:then
          (Audio|Components|Audio|FadeOut _inairaudiocomp 0.2 0.0 "Linear" :!class /Script/Engine.AudioComponent))
        (:else)))))

(event EventOnDamageServer
  (params (Damage :type Float%20%28single-precision%29) (DamageTags :type Gameplay%20Tag%20Container%20Structure%20%28by%20ref%29) (Momentum :type Vector :default "0, 0, 0") (HitInfo :type Hit%20Result%20Structure%20%28by%20ref%29) (InstigatedBy :type Controller%20Object%20Reference) (DamageCauser :type Actor%20Object%20Reference) (EffectContext :type Gameplay%20Effect%20Context%20Handle%20Structure))
  (bind _ghostcontroller (Variables|Default|GetGhostController))
  (|CallParentFunction Damage DamageTags Momentum HitInfo InstigatedBy DamageCauser EffectContext)
  (if (<= (FortPawn|GetHealth) 0)
    (Ability|GameplayEffect|AddLooseGameplayTags self (GameplayTags|MakeLiteralGameplayTagContainer "(GameplayTags=((TagName=\"PuppeteerBoss.Ghost.DefeatedByDmg\")))"))
    (else
      _))
  (bind _asfort_player_pawn_athena (Utilities|Casting|CastToFortPlayerPawnAthena (Actor|GetOwner DamageCauser)))
  (if (ScalableFloat|GetValueasBool (Variables|Hotfix|GetTriggerAttackOnDmg_Hf))
    (if (not (GameplayTags|HasMatchingGameplayTag "(TagName=\"PuppeteerBoss.Ghost.IsAttacking\")"))
      (Ability|GameplayEffect|AddLooseGameplayTags self (GameplayTags|MakeLiteralGameplayTagContainer "(GameplayTags=((TagName=\"PuppeteerBoss.Ghost.IsAttacking\")))"))
      (Class|FortAthenaAIBotController|AddTargetInfo _ghostcontroller _asfort_player_pawn_athena true "Threat_Seeing" 5.0 2e+03)
      (AI|SetFocus _ghostcontroller _asfort_player_pawn_athena)
      (Abilities|TryActivateAbilitiesbyTag (Ability|GetAbilitySystemComponent self) (Variables|Default|GetGhostAttackAbilityTag) true))))

(event Custom|CheckGhostIsInLeash
  (if (not (Class|FortAthenaAIBotController|IsLocationInsideLeash (Variables|Default|GetGhostController) (Transformation|GetActorLocation) false 0.0))
    (CallFunction|GhostAttackFinished)))

(fn Hide/Unhide
  (params (Hide :type Boolean))
  (bind _returnvalue (Utilities|Time|GetGameTimeinSeconds))
  (bind _mesh (Variables|Character|GetMesh))
  (if Hide
    (Rendering|Material|SetCustomPrimitiveDataVector2 _mesh 0 (Math|Vector2D|MakeVector2D _returnvalue))
    (else
      (Rendering|Material|SetCustomPrimitiveDataVector2 _mesh 0 (Math|Vector2D|MakeVector2D _returnvalue 1.0)))))

(fn OnRep_bGhostAudioShouldStop
  (Networking|SwitchHasAuthority
    (:Remote
      (if (not (Variables|Audio|GetbGhostAudioShouldStop))
        (CallFunction|FadeOutGhostLoop)))))

(fn OnRep_PlayerHasDirectLOS)

(fn OnRep_ShouldHide
  (Networking|SwitchHasAuthority
    (:Remote
      (CallFunction|Hide/Unhide (Variables|Default|GetShouldHide)))))

(fn OnRep_TargetPlayer
  (bind _returnvalue (Pawn|GetController))
  (bind _returnvalue_2 (Actor|GetComponentbyClass _returnvalue "/Script/AIModule.BlackboardComponent"))
  (bind _targetplayer (Variables|Default|GetTargetPlayer))
  (Networking|SwitchHasAuthority
    (:Authority
      (Utilities|IsValid _returnvalue
        (:Is%20Valid
          (Utilities|IsValid _returnvalue_2
            (:Is%20Valid
              (AI|Components|Blackboard|SetValueasObject _returnvalue_2 (Variables|Default|GetGhostTargetActorKeyName) _targetplayer))
            (:Is%20Not%20Valid
              _)))
        (:Is%20Not%20Valid
          _))
      (Utilities|IsValid _targetplayer
        (:Is%20Valid
          (CallFunction|ResetWaitforDBNOOrDeath)
          (CallFunction|WaitforDBNOOrDeath)
          (if (not (GameplayTags|HasMatchingGameplayTag _targetplayer "(TagName=\"PuppeteerBoss.Player.IsHaunted\")"))
            (Ability|GameplayEffect|AddLooseGameplayTags _targetplayer (GameplayTags|MakeGameplayTagContainerfromTag "(TagName=\"PuppeteerBoss.Player.IsHaunted\")"))))))
    (:Remote
      (Audio|Components|Audio|FadeIn (Variables|Sounds|GetInAirAudioComp) 1.0 1.0 0.0 "Linear" :!class /Script/Engine.AudioComponent))))

(fn StopHauntingTarget
  (params (TargetPlayer :type Fort%20Player%20Pawn%20Object%20Reference))
  (Utilities|IsValid TargetPlayer
    (:Is%20Valid
      (Ability|GameplayEffect|RemoveLooseGameplayTags TargetPlayer (GameplayTags|MakeGameplayTagContainerfromTag "(TagName=\"PuppeteerBoss.Player.IsHaunted\")"))
      (Environment|ApplyGlobalEnvironmentGameplayEffecttoActor TargetPlayer "/PuppeteerBoss/GameplayEffect/GE_Puppeteer_GhostHauntCooldown.GE_Puppeteer_GhostHauntCooldown_C")
      (Variables|Default|SetTargetPlayer 0))))

(fn TieLeashToGhostSpawnArea
  (bind (_returnvalue_x _returnvalue_y) (Math|Vector|BreakVector (Components|Box|GetScaledBoxExtent (Variables|/PuppeteerBoss/Box/Puppeteer_GhostSpawnArea_Box.Puppeteer_GhostSpawnArea_Box_C|GetBox (Variables|Default|GetGhost%20Spawn%20Area)))))
  (bind _ghost_spawn_area (Variables|Default|GetGhost%20Spawn%20Area))
  (bind _ghostcontroller (Variables|Default|GetGhostController))
  (Utilities|IsValid _ghost_spawn_area
    (:Is%20Valid
      (Class|BPFLAICommonFunctions|SetRadialLeash _ghostcontroller (Transformation|GetWorldLocation (Variables|/PuppeteerBoss/Box/Puppeteer_GhostSpawnArea_Box.Puppeteer_GhostSpawnArea_Box_C|GetBox _ghost_spawn_area)) (Math|Float|SelectFloat _returnvalue_x _returnvalue_y (> _returnvalue_x _returnvalue_y)) (+ (Math|Float|SelectFloat _returnvalue_x _returnvalue_y (> _returnvalue_x _returnvalue_y)) 3e+03)))
    (:Is%20Not%20Valid
      (Class|BPFLAICommonFunctions|SetRadialLeash _ghostcontroller (Transformation|GetActorLocation) 7e+03 1e+04))))

(fn UpdateStrikeCollisionBox
  (bind _strikecollision (Variables|Default|GetStrikeCollision))
  (bind (_boxextent_x _boxextent_y) (Math|Vector|BreakVector (Variables|/Script/Engine.BoxComponent|GetBoxExtent_X _strikecollision)))
  (Components|Box|SetBoxExtent _strikecollision (Math|Vector|MakeVector (ScalableFloat|GetValueatLevel (Variables|Hotfix|GetStrikeCollisionBoxExtentX_Hf)) (ScalableFloat|GetValueatLevel (Variables|Hotfix|GetStrikeCollisionBoxExtentY_Hf)) (ScalableFloat|GetValueatLevel (Variables|Hotfix|GetStrikeCollisionBoxExtentZ_Hf))) false)
  (Transformation|AddRelativeLocation _strikecollision (Math|Vector|MakeVector 0.0 _boxextent_y) false false))

(fn UpdateStrikeCollisionSphere
  (bind _strikecollisionsphereradius_hf (Variables|Hotfix|GetStrikeCollisionSphereRadius_Hf))
  (bind _returnvalue (ScalableFloat|GetValueatLevel _strikecollisionsphereradius_hf))
  (bind _strikecollisionsphere (Variables|Default|GetStrikeCollisionSphere))
  (Components|Sphere|SetSphereRadius _strikecollisionsphere _returnvalue false)
  (Transformation|AddRelativeLocation _strikecollisionsphere (Math|Vector|MakeVector 0.0 0.0 (+ _returnvalue 5.0)) false false))

(fn ConstructionScript
  (|CallParentFunction)
  (CallFunction|UpdateStrikeCollisionBox)
  (CallFunction|UpdateStrikeCollisionSphere))
