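; AI random-targeter controller component for the Puppeteer Demon boss, supplying the rules for which targets it may pick and when it may switch.
; On BeginPlay it caches its owning Athena bot controller.
; Switching is always allowed when the blackboard's inside-leash flag is false, otherwise it requires CanSwitchTarget and the scalable-float randomizer toggle.
; Non-player and DBNO targets are ignored, and targets outside the leash are undesirable.
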(event EventBeginPlay
  (bind _asfort_athena_aibot_controller (Utilities|Casting|CastToFortAthenaAIBotController (Components|GetOwner)))
  (Variables|Default|SetOwnerController _asfort_athena_aibot_controller))

(fn AllowedToSwitchTarget
  (if (not (AI|Components|Blackboard|GetValueasBool (Variables|/Script/AIModule.AIController|GetBlackboard (Variables|Default|GetOwnerController)) (Utilities|Name|MakeLiteralName "AIEvaluator_TargetContext_InsideLeash")))
    (return true)
    (else
      (return (and (Variables|Default|GetCanSwitchTarget) (ScalableFloat|GetValueasBool (Variables|Default|GetTargetRandomizerEnabled_Hf)))))))

(fn TargetIsIgnored
  (params (Target :type Actor%20Object%20Reference))
  (bind _asfort_player_pawn_athena (Utilities|Casting|CastToFortPlayerPawnAthena Target)
    (:then
      (if (Health|IsDBNO _asfort_player_pawn_athena)
        (return true)
        (else
          (return false))))
    (:CastFailed
      (return true))))

(fn TargetIsUndesirable
  (params (Target :type Actor%20Object%20Reference))
  (return (not (Class|FortAthenaAIBotController|IsLocationInsideLeash (Variables|Default|GetOwnerController) (Transformation|GetActorLocation Target) false 0.0))))
